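# Prototype master material for Asteria quarry blocks, Nanite-enabled, built from world-aligned marble with an optional baked normal and a gradient tint.
# A glowing "Fortnitanium" layer is masked by noise and marble textures, then dirt is layered on upward-facing surfaces.
# An optional branch blends landscape runtime virtual texture color and normals in by height above ground.
# Skip RVT is on by default, and Forward and Mobile paths always bypass that blend.
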Material /Game/Environments/Asteria/Sets/Quarry/Material/old/M_Asteria_Quarry_Block_Proto
graph {
 properties {bUseMaterialAttributes=true, bUsedWithNanite=true, pixelDepthOffsetMode=PDOM_Legacy, usedShadingModels=Default Lit}
 nodes [
  n1 = ScalarParameter() {defaultValue=1200, parameterName=Marble Tiling}
  n2 = MaterialFunctionCall!(
   Diffuse Texture Object=TextureObject() {texture=/Game/Creative/Sets/PropSets/Primitives/Customization/Textures/Marble/T_CP_Marble_BC2}
   Spec Texture Object=TextureObject() {samplerType=SAMPLERTYPE_Masks, texture=/Game/Creative/Sets/PropSets/Primitives/Customization/Textures/Marble/T_CP_Marble_S}
   TextureSize=@n1
   --------------------------
   Use seperate settings for Z projection
   Use seperate texture for Z projection
   Texture Size Z projection
   Z Plane Diffuse Texture Object
   Z Plane Spec Texture Object
   "--------------------------      "
   ProjectionTransitionContrast
   Use Texture Alpha Heightmap to Refine the Transistion
   HeightLerpTransitionContrast
   PickTheDiffuseHeightChannel
   TheDiffuseHeightChannel
   "--------------------------       "
   NormalTextureObject=TextureObject() {samplerType=SAMPLERTYPE_Normal, texture=/Game/Creative/Sets/PropSets/Primitives/Customization/Textures/Marble/T_CP_Marble_N}
   World Space Normals=StaticBool()
   "--------------------------  "
   Export Float 4 Diffuse
   Export Float 4 Spec
   "--------------------------     "
   WorldPosition
   World Space Normal
  ) {materialFunction=/Engine/Functions/Engine_MaterialFunctions02/Texturing/WorldAlignedTextures_Complex}
  n3 = MaterialFunctionCall(Float3=@n2.XYZ Spec) {materialFunction=&BreakOutFloat3Components}
  n4 = TextureSampleParameter2D() {parameterName=Baked Normal, samplerType=SAMPLERTYPE_Normal, texture=/Game/Environments/Apollo/Rocks/Rockslide/Textures/T_Apollo_RockSlide_N}
  n5 = MakeMaterialAttributes!(
   BaseColor=ComponentMask(None=@n2.XYZ Diffuse Texture) {b=true, g=true, r=true}
   Metallic=@n3.G
   Specular=@n3.G
   Roughness=@n3.G
   Anisotropy
   EmissiveColor
   Opacity
   OpacityMask
   Normal=MaterialFunctionCall(
    Normal=StaticSwitchParameter(True=@n4.RGB, False=@n2.Normal XYZ Texture) {parameterName=Use Baked Normal}
    Flatness=ScalarParameter() {defaultValue=-1, parameterName=Normal Flatness}
   ) {materialFunction=&FlattenNormal}.Result
   Tangent
   WorldPositionOffset
   SubsurfaceColor
   ClearCoat
   ClearCoatRoughness
   AmbientOcclusion
   Refraction
   CustomizedUVs_0
   CustomizedUVs_1
   CustomizedUVs_2
   CustomizedUVs_3
   CustomizedUVs_4
   CustomizedUVs_5
   CustomizedUVs_6
   CustomizedUVs_7
   PixelDepthOffset
   ShadingModel
   Displacement
  )
  n6 = VectorParameter() {group=Fortnitanium, parameterName=Fortnitanium_Color, defaultValue=<rgba 0.0052 0.483 1 1>}
  n7 = ScalarParameter() {defaultValue=0.5, group=Fortnitanium, parameterName=Fortnitanium_Metalness}
  n8 = ScalarParameter() {defaultValue=1, group=Fortnitanium, parameterName=Fortnitanium_Specular}
  n9 = ScalarParameter() {defaultValue=0.2, group=Fortnitanium, parameterName=Fortnitanium_Roughness}
  n10 = Multiply(A=ScalarParameter() {defaultValue=3, group=Fortnitanium, parameterName=Fortnitanium_Emissive}, B=@n6.RGB)
  n11 = MaterialFunctionCall(
   In=MaterialFunctionCall(
    TextureObject=TextureObject() {samplerType=SAMPLERTYPE_Masks, texture=/Game/Weapons/WeaponSkins/Wraps/Textures/Mash/Mash_Noise}
    TextureSize=ScalarParameter() {defaultValue=10000, group=Fortnitanium, parameterName=Fortnitanium_Noise_Tiling}
   ) {materialFunction=/Game/Weapons/WeaponSkins/Wraps/MaterialFunctions/WorldAlignedTexture}.XYZ Texture
   Contrast=Constant() {r=0.5}
  ) {materialFunction=&CheapContrast}
  n12 = MaterialFunctionCall(
   In=MaterialFunctionCall(
    TextureObject=TextureObject() {texture=/Game/Creative/Sets/PropSets/Primitives/Customization/Textures/Marble/T_CP_Marble_BC3}
    TextureSize=@n1
   ) {materialFunction=&WorldAlignedTexture}.XYZ Texture
   Contrast=Constant() {r=0.1}
  ) {materialFunction=&CheapContrast}
  n13 = MaterialFunctionCall(
   Diffuse Texture Object=TextureObject() {texture=/Game/Environments/Apollo/Sets/Terrain/Textures/T_Apollo_DirtHeight_D}
   Spec Texture Object=TextureObject() {samplerType=SAMPLERTYPE_Masks, texture=/Game/Environments/Apollo/Sets/Terrain/Textures/T_Apollo_Dirt_S_01}
   TextureSize=Multiply(A=ScalarParameter() {defaultValue=1500, parameterName=Dirt Tiling}, B=Constant3Vector() {constant=<rgba 1 1 1 0>})
   NormalTextureObject=TextureObject() {samplerType=SAMPLERTYPE_Normal, texture=/Game/Environments/Apollo/Sets/Terrain/Textures/T_Apollo_Dirt_N_01}
   Export Float 4 Diffuse=StaticBool() {value=true}
  ) {materialFunction=&WorldAlignedTextures_Complex}
  n14 = MaterialFunctionCall(Float3=@n13.XYZ Spec) {materialFunction=&BreakOutFloat3Components}
  n15 = MaterialFunctionCall!(
   Base Material=MaterialFunctionCall(
    Base Material=MaterialFunctionCall!(transitionFactor, MaterialIn=@n5) {materialFunction=/Game/Athena/WinterArt/Materials/MF_WSGradient}.Color*Gradient
    Top Material=MakeMaterialAttributes(BaseColor=@n6.RGB, Metallic=@n7, Specular=@n8, Roughness=@n9, EmissiveColor=@n10)
    Alpha=LinearInterpolate(B=@n11.Result, Alpha=@n12.Result)
   ) {materialFunction=/Engine/Functions/MaterialLayerFunctions/MatLayerBlend_Standard}.Blended Material
   Top Material=MakeMaterialAttributes(
    BaseColor=ComponentMask(None=@n13.XYZ Diffuse Texture) {b=true, g=true, r=true}
    Metallic=@n14.G
    Specular=@n14.G
    Roughness=@n14.G
    Normal=@n13.Normal XYZ Texture
   )
   Alpha=MaterialFunctionCall!(
    In Explicit Normal=@n2.Normal XYZ Texture
    In World Vector=Constant3Vector() {constant=<rgba 0 0 1 0>}
    Blend Sharpness=ScalarParameter() {defaultValue=2.0002, parameterName=Blend_Sharpness}
    Blend Bias=ScalarParameter() {parameterName=Blend_Bias}
    Clamped?
    Alpha=MaterialFunctionCall(Float4=@n13.XYZ Diffuse Texture) {materialFunction=&BreakOutFloat4Components}.A
   ) {materialFunction=/Engine/Functions/Engine_MaterialFunctions01/AlphaBlend/WorldAlignedBlend}.w/Vertex Normals
  ) {materialFunction=&MatLayerBlend_Standard}
  n16 = RuntimeVirtualTextureSampleParameter() {materialType=BaseColor_Normal_Specular_YCoCg, parameterName=RVT, virtualTexture=/Game/Athena/Apollo/Maps/Landscape/VT/RVT_Apollo_Landscape}
  n17 = WorldPosition()
  n18 = RuntimeVirtualTextureSampleParameter() {materialType=WorldHeight, parameterName=RVTHeight, virtualTexture=/Game/Athena/Apollo/Maps/Landscape/VT/RVT_Apollo_Landscape_Height}
  n19 = Subtract(A=Add(A=ScalarParameter() {group=RVT, parameterName=Offset RVT WP}, B=ComponentMask(None=@n17.XYZ) {b=true}), B=@n18.WorldHeight)
  n20 = Divide(
   A=NamedRerouteDeclaration(None=@n19) {name=RVT Height, nodeColor=<rgba 0.6875 0 1 1>}
   B=ScalarParameter() {defaultValue=25, group=RVT, parameterName=Blend RVT Normals}
  )
  n21 = Saturate(None=@n20)
  n22 = BlendMaterialAttributes(
   A=SetMaterialAttributes!(MaterialAttributes=@n15.Blended Material, Normal=@n16.Normal) {attributeSetTypes [#1]}
   B=@n15.Blended Material
   Alpha=Saturate(None=StaticSwitchParameter(True=@n21, False=@n20) {group=Force RVT, parameterName=Use Force RVT Mask, sortPriority=0})
  ) {vertexAttributeBlendType=UseA}
  n23 = VertexNormalWS()
  n24 = SetMaterialAttributes!(
   MaterialAttributes=@n22
   Base Color=LinearInterpolate(
    A=GetMaterialAttributes(None=@n15.Blended Material) {attributeGetTypes [#2, #1]}.Base Color
    B=@n16.BaseColor
    Alpha=ComponentMask(None=@n23) {b=true}
   )
   Specular=@n16.Specular
   Roughness=@n16.Specular
   Emissive Color=Constant()
  ) {attributeSetTypes [
    #2
    #3
    #4
    #5
   ]}
  n25 = Divide(A=@n19, B=ScalarParameter() {defaultValue=55, group=RVT, parameterName=Blend RVT})
  n26 = Saturate(None=StaticSwitchParameter(True=Saturate(None=@n25), False=@n25) {group=Force RVT, parameterName=Use Force RVT Mask, sortPriority=0})
  n27 = MaterialFunctionCall(
   NoRVT=@n15.Blended Material
   RVT=MaterialProxyReplace(Realtime=BlendMaterialAttributes(A=@n24, B=@n22, Alpha=@n26) {vertexAttributeBlendType=UseA}, MaterialProxy=@n15.Blended Material)
  ) {materialFunction=&FN_RVT_Switch_Material}
  StaticSwitchParameter(
   True=@n15.Blended Material
   False=ShadingPathSwitch(Default=@n27.Result, Deferred=@n27.Result, Forward=@n15.Blended Material, Mobile=@n15.Blended Material)
  ) {
   defaultValue=true
   desc="If for some reason you want to use this material but hate to use RVT ( or you'r Bryce L) Will automatically convert details to use up vector instead distance from the ground "
   group=RVT
   parameterName=Skip RVT
   sortPriority=-4
  }
 ]
}
