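# Prototype Asteria Quarry mesh material that blends props into the Apollo landscape through Runtime Virtual Textures.
# World-aligned Set1 takes a gradient and sediment tint, with an optional Set2 layer, distance-field AO tint and height-lerped transition texture.
# A static switch adds a baked normal, flattened under Nanite, before tangent-to-world conversion.
# RVT color, normal and specular blend in by height above landscape and vertex normal, skipped on forward, mobile and proxy paths.
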Material /Game/Environments/Asteria/Sets/Quarry/Material/old/M_Asteria_Quarry_RVTBlend_Proto
graph {
 properties {
  bTangentSpaceNormal=false
  bUseMaterialAttributes=true
  bUsedWithInstancedStaticMeshes=true
  bUsedWithNanite=true
  bUsedWithSplineMeshes=true
  pixelDepthOffsetMode=PDOM_Legacy
  usedShadingModels=Default Lit
 }
 nodes [
  n1 = VectorParameter() {group=Transition Texture, parameterName=Transition Tint, sortPriority=-1, defaultValue=<rgba 0.4271 0.3646 0.3181 0>}
  n2 = VectorParameter() {group=Transition Texture Edge, parameterName=Transition Tint2, sortPriority=-1, defaultValue=<rgba 0.4271 0.3646 0.3181 0>}
  n3 = NamedRerouteUsage()
  n4 = ScalarParameter() {defaultValue=39.6983, group=Transition Texture Edge, parameterName=Transition Color Fade}
  n5 = StaticBoolParameter() {group=Transition Texture, parameterName=Transition use mesh UV}
  n6 = StaticSwitch(True=NamedRerouteUsage(), False=NamedRerouteUsage(), Value=@n5)
  n7 = OneMinus(None=ComponentMask(None=@n6) {a=true})
  n8 = MaterialFunctionCall(
   A=@n1.RGB
   B=@n2.RGB
   Transition Phase=Saturate(None=Divide(A=@n3, B=@n4))
   Height Texture=@n7
   Contrast=ScalarParameter() {defaultValue=1, group=Transition Texture Edge, parameterName=Transition Color Fade Contrast}
  ) {materialFunction=&HeightLerp}
  n9 = ComponentMask(None=@n6) {r=true}
  n10 = LinearInterpolate(
   A=ScalarParameter() {group=Transition Texture, parameterName=Transition Texture Rough Min, sortPriority=2}
   B=ScalarParameter() {defaultValue=2.5, group=Transition Texture, parameterName=Transition Texture Rough Max, sortPriority=2}
   Alpha=@n9
  )
  n11 = Multiply(A=@n10, B=ScalarParameter() {defaultValue=5, group=Transition Texture Edge, parameterName=Transition Edge Rough})
  n12 = NamedRerouteUsage()
  n13 = ScalarParameter() {defaultValue=1, group=Transition Texture, parameterName=Transition Normals Value, sortPriority=1}
  n14 = StaticSwitchParameter(
   True=LinearInterpolate(
    A=@n13
    B=ScalarParameter() {defaultValue=1, group=Transition Texture Edge, parameterName=Transition Normals Edge Value, sortPriority=1}
    Alpha=@n12
   )
   False=@n13
  ) {group=Transition Texture Edge, parameterName="Transition Edge "}
  n15 = Multiply(
   A=StaticSwitch(True=NamedRerouteUsage(), False=NamedRerouteUsage(), Value=@n5)
   B=AppendVector(A=AppendVector(A=@n14, B=@n14), B=Constant() {r=1})
  )
  n16 = TextureObjectParameter() {group=Set1, parameterName=Specular Masks, samplerType=SAMPLERTYPE_Masks, texture=&T_Fortnite_Default_S}
  n17 = TextureObjectParameter() {
   group=Set1
   parameterName=Normals
   samplerType=SAMPLERTYPE_Normal
   texture=/Game/Environments/Apollo/Sets/Terrain/Textures/T_Apollo_Sand_N_01
  }
  n18 = StaticBool() {value=true}
  n19 = MaterialFunctionCall!(
   Diffuse Texture Object=TextureObjectParameter() {group=Set1, parameterName=Diffuse, texture=/Game/Global/Textures/Default/Blanks/T_MidgreysRGB}
   Spec Texture Object=@n16
   TextureSize=ScalarParameter() {defaultValue=2048, group=Set1, parameterName=Set1 Tiling}
   --------------------------
   Use seperate settings for Z projection
   Use seperate texture for Z projection
   Texture Size Z projection
   Z Plane Diffuse Texture Object
   Z Plane Spec Texture Object
   "--------------------------      "
   ProjectionTransitionContrast
   Use Texture Alpha Heightmap to Refine the Transistion
   HeightLerpTransitionContrast
   PickTheDiffuseHeightChannel
   TheDiffuseHeightChannel
   "--------------------------       "
   NormalTextureObject=@n17
   World Space Normals=StaticBool()
   "--------------------------  "
   Export Float 4 Diffuse=@n18
   Export Float 4 Spec=@n18
   "--------------------------     "
   WorldPosition
   World Space Normal
  ) {materialFunction=/Engine/Functions/Engine_MaterialFunctions02/Texturing/WorldAlignedTextures_Complex}
  n20 = ScalarParameter() {defaultValue=1, group=Set1, parameterName=Normals Value, sortPriority=2}
  n21 = Multiply(A=@n19.Normal XYZ Texture, B=AppendVector(A=AppendVector(A=@n20, B=@n20), B=Constant() {r=1}))
  n22 = Add(A=@n15, B=Multiply(A=@n21, B=Constant3Vector() {constant=<rgba 1 1 0 0>}))
  n23 = SetMaterialAttributes!(
   MaterialAttributes
   Base Color=Multiply(
    A=StaticSwitchParameter(True=@n8.Results, False=@n1.RGB) {group=Transition Texture Edge, parameterName="Transition Edge "}
    B=ComponentMask(None=@n6) {b=true, g=true, r=true}
   )
   Specular=Multiply(A=@n9, B=ScalarParameter() {defaultValue=0.5, group=Transition Texture, parameterName=Transition Texture Spec, sortPriority=2})
   Roughness=StaticSwitchParameter(True=LinearInterpolate(A=@n10, B=@n11, Alpha=@n12), False=@n10) {group=Transition Texture Edge, parameterName="Transition Edge "}
   Normal=StaticSwitchParameter(True=LinearInterpolate(A=@n15, B=@n22, Alpha=@n12), False=@n15) {group=Transition Texture Edge, parameterName="Transition Edge "}
  ) {attributeSetTypes [
    #1
    #2
    #3
    #4
   ]}
  NamedRerouteDeclaration(None=@n23) {name=Transition, nodeColor=<rgba 0.5 0 1 1>}
  n24 = TextureSampleParameter2D() {group=Set1, parameterName=Diffuse, samplerSource=SSM_Wrap_WorldGroupSettings, texture=&T_MidgreysRGB}
  n25 = LinearInterpolate(A=Constant() {r=0.65}, Alpha=@n24.R) {constA=0.5}
  n26 = NamedRerouteUsage()
  n27 = MaterialFunctionCall(
   Transition Phase=Divide(
    A=Multiply(A=@n25, B=@n26)
    B=ScalarParameter() {defaultValue=65, group=Transition Texture, parameterName=Transition Texture Fade, sortPriority=4}
   )
   Height Texture=Multiply(A=@n7, B=ScalarParameter() {defaultValue=1, group=Transition Texture, parameterName=Transition Texture Alpha, sortPriority=3})
   Contrast=ScalarParameter() {defaultValue=3, group=Transition Texture, parameterName=Transition Texture Contrast, sortPriority=4}
  ) {materialFunction=&HeightLerp}
  NamedRerouteDeclaration(None=Min(A=@n27.Alpha, B=@n27.Alpha)) {name=Transition Mask, nodeColor=<rgba 0.3125 0 1 1>}
  n28 = TextureCoordinate()
  n29 = Divide(A=ObjectRadius(), B=ScalarParameter() {defaultValue=512, group=Set2, parameterName=ObjRadiusScaleSet2, sortPriority=-2})
  n30 = TextureCoordinate()
  n31 = ScalarParameter() {defaultValue=1, group=Set2, parameterName=ScaleSet2UV, sortPriority=-2}
  n32 = NamedRerouteDeclaration(None=StaticSwitchParameter(True=Multiply(A=@n28, B=@n29), False=Multiply(A=@n30, B=@n31)) {group=Set2, parameterName=Set 2 Use ObjRadius UvScale, sortPriority=-2}) {name=Set 2 UV, nodeColor=<rgba 0 1 0.0391 1>}
  n33 = TextureObjectParameter() {group=Set2, parameterName=Diffuse 2, sortPriority=-2, texture=/Game/Packages/DS_Fortnite_Terrain_NoLOD/Textures/T_Grasslands_AD_D}
  n34 = TextureObjectParameter() {
   group=Set2
   parameterName=Spec2
   samplerType=SAMPLERTYPE_Masks
   sortPriority=-2
   texture=/Game/Environments/Apollo/Sets/Terrain/Textures/T_Apollo_Dirt_S_02
  }
  n35 = TextureObjectParameter() {
   group=Set2
   parameterName=Normals 2
   samplerType=SAMPLERTYPE_Normal
   sortPriority=-2
   texture=/Game/Packages/DS_Fortnite_Terrain_NoLOD/Textures/T_Grasslands_N
  }
  n36 = MaterialFunctionCall(
   Diffuse Texture Object=@n33
   Spec Texture Object=@n34
   TextureSize=ScalarParameter() {defaultValue=512, group=Set2, parameterName=ScaleSet2, sortPriority=-2}
   NormalTextureObject=@n35
   World Space Normals=StaticBool()
   Export Float 4 Diffuse=StaticBool() {value=true}
  ) {materialFunction=&WorldAlignedTextures_Complex}
  n37 = StaticSwitchParameter(True=TextureSample(UVs=@n32, Tex=@n33).RGB, False=@n36.Z Diffuse Texture) {group=Set2, parameterName=Set 2 Use UV Coord, sortPriority=-2}
  n38 = Multiply(
   A=VectorParameter() {group=Set2, parameterName=Set2 Tint, defaultValue=<rgba 1 1 1 0>}.RGB
   B=ComponentMask(None=@n37) {b=true, g=true, r=true}
  )
  n39 = MaterialFunctionCall(Float3=StaticSwitchParameter(True=TextureSample(UVs=@n32, Tex=@n34).RGB, False=@n36.Z Spec) {group=Set2, parameterName=Set 2 Use UV Coord, sortPriority=-2}) {materialFunction=&BreakOutFloat3Components}
  n40 = ScalarParameter() {defaultValue=1, group=Set2, parameterName=Set2 Normals Value, sortPriority=2}
  n41 = Multiply(
   A=StaticSwitchParameter(True=TextureSample(UVs=@n32, Tex=@n35).RGB, False=@n36.Normal Z Texture) {group=Set2, parameterName=Set 2 Use UV Coord, sortPriority=-2}
   B=AppendVector(A=AppendVector(A=@n40, B=@n40), B=Constant() {r=1})
  )
  n42 = SetMaterialAttributes!(MaterialAttributes, Base Color=@n38, Specular=@n39.R, Roughness=@n39.R, Normal=@n41, Opacity, Metallic=@n39.R) {attributeSetTypes [
    #1
    #2
    #3
    #4
    #5
    #6
   ]}
  n43 = ComponentMask(None=@n37) {r=true}
  n44 = SetMaterialAttributes!(
   MaterialAttributes
   Base Color=@n38
   Specular=@n43
   Roughness=LinearInterpolate(
    A=ScalarParameter() {defaultValue=0.6, group=Set2, parameterName=Set2 Rough Min}
    B=ScalarParameter() {defaultValue=0.6, group=Set2, parameterName=Set2 Rough Max}
    Alpha=@n43
   )
   Normal=@n41
   Opacity=LinearInterpolate(
    A=ScalarParameter() {group=Set2, parameterName=Set2AlphaThreshold Min}
    B=ScalarParameter() {defaultValue=1, group=Set2, parameterName=Set2AlphaThreshold Max}
    Alpha=ComponentMask(None=@n36.Z Diffuse Texture) {a=true}
   )
  ) {attributeSetTypes [
    #1
    #2
    #3
    #4
    #5
   ]}
  n45 = ScalarParameter() {defaultValue=3, parameterName=Blend Sharpness}
  n46 = ScalarParameter() {defaultValue=-0.5, parameterName=Blend Bias}
  n47 = StaticBool() {value=true}
  NamedRerouteDeclaration(None=BlendMaterialAttributes(
    A=NamedRerouteUsage()
    B=StaticSwitchParameter(True=@n42, False=@n44) {desc=DiffuseSpecNormals true, group=Set2, parameterName=Set2DSNTrue}
    Alpha=MaterialFunctionCall!(In Explicit Normal, In World Vector, Blend Sharpness=@n45, Blend Bias=@n46, Clamped?=@n47, Alpha) {materialFunction=/Engine/Functions/Engine_MaterialFunctions01/AlphaBlend/WorldAlignedBlend}.w/ Explicit Normal
   ) {vertexAttributeBlendType=UseA}) {name=Set2, nodeColor=<rgba 1 0 0.6562 1>}
  n48 = TextureObjectParameter() {
   group=Transition Texture
   parameterName=Transition Texture Diffuse
   sortPriority=0
   texture=/Game/Environments/Apollo/Sets/Terrain/Textures/T_Apollo_Dirt_D_01
  }
  n49 = TextureObjectParameter() {
   group=Transition Texture
   parameterName=Transition Texture Normal
   samplerType=SAMPLERTYPE_Normal
   sortPriority=0
   texture=/Game/Environments/Apollo/Sets/Terrain/Textures/T_Apollo_Dirt2_N
  }
  n50 = MaterialFunctionCall(
   Diffuse Texture Object=@n48
   TextureSize=ScalarParameter() {defaultValue=512, group=Transition Texture, parameterName=Transition Scale, sortPriority=1}
   NormalTextureObject=@n49
   World Space Normals=StaticBool()
   Export Float 4 Diffuse=StaticBool() {value=true}
  ) {materialFunction=&WorldAlignedTextures_Complex}
  NamedRerouteDeclaration(None=@n50.Z Diffuse Texture) {name=Transition Diffuse WSZ, nodeColor=<rgba 1 0 0.5156 1>}
  NamedRerouteDeclaration(None=@n50.Normal Z Texture) {name=Transition Normal WSZ, nodeColor=<rgba 1 0.9609 0 1>}
  n51 = Multiply(
   A=TextureCoordinate()
   B=Divide(A=ObjectRadius(), B=ScalarParameter() {defaultValue=512, group=Transition Texture, parameterName=Transition Scale, sortPriority=1})
  )
  NamedRerouteDeclaration(None=TextureSample(UVs=@n51, Tex=@n48).RGBA) {name=Transition Diffuse UV, nodeColor=<rgba 1 0.1875 0 1>}
  NamedRerouteDeclaration(None=TextureSample(UVs=@n51, Tex=@n49) {samplerType=SAMPLERTYPE_Normal}.RGB) {name=Transition Normal UV, nodeColor=<rgba 0 0.9297 1 1>}
  NamedRerouteDeclaration(None=@n8.Alpha) {name=TrColorFadeMask, nodeColor=<rgba 0 1 0.1562 1>}
  NaniteReplace!(Default, Nanite)
  n52 = ComponentMask(None=@n19.XYZ Diffuse Texture) {b=true, g=true, r=true}
  n53 = VectorParameter() {group=Set1, parameterName=Set Tint, sortPriority=-1, defaultValue=<rgba 1 1 1 0>}
  n54 = Multiply(A=@n53.RGB, B=@n52)
  n55 = ComponentMask(None=@n19.XYZ Diffuse Texture) {a=true}
  n56 = MaterialFunctionCall(Float4=@n19.XYZ Spec) {materialFunction=&BreakOutFloat4Components}
  n57 = TextureSampleParameter2D(UVs=StaticSwitchParameter(True=TextureCoordinate() {coordinateIndex=1}, False=NamedRerouteUsage()) {group=Set1, parameterName=Custom Emissive UV2}) {
   group=Set1
   parameterName=Custom Emissive
   samplerSource=SSM_Wrap_WorldGroupSettings
   sortPriority=0
   texture=&T_MidgreysRGB
  }
  n58 = SetMaterialAttributes!(
   MaterialAttributes
   Base Color=StaticSwitchParameter(True=LinearInterpolate(A=@n52, B=@n54, Alpha=@n55), False=Multiply(A=@n53.RGB, B=@n52)) {group=Set1, parameterName=use Diffuse Alpha as color Mask, sortPriority=-2}
   Specular=Multiply(A=ScalarParameter() {defaultValue=1, group=Set1, parameterName=Specular, sortPriority=1}, B=@n56.R)
   Metallic=@n56.G
   Roughness=Multiply(A=@n56.B, B=ScalarParameter() {defaultValue=1, group=Set1, parameterName=Roughness, sortPriority=1})
   Normal=@n21
   Opacity=@n55
   Emissive Color=Multiply(
    A=StaticSwitchParameter(True=@n57.RGB, False=@n56.A) {group=Set1, parameterName=Custom Emissive}
    B=VectorParameter() {desc="Emissive should be in spec alpha ", group=Set1, parameterName=EmissiveColor}.RGB
   )
  ) {attributeSetTypes [
    #1
    #2
    #6
    #3
    #4
    #5
    #7
   ]}
  n59 = MaterialFunctionCall!(transitionFactor, MaterialIn=@n58) {materialFunction=/Game/Athena/WinterArt/Materials/MF_WSGradient}
  n60 = LinearInterpolate(
   A=Constant3Vector() {constant=<rgba 0.196 0.2859 0.3333 0>}
   B=Constant3Vector() {constant=<rgba 1 1 1 0>}
   Alpha=MaterialFunctionCall(
    TextureObject=TextureObject() {texture=/Game/Environments/Apollo/Sets/Abducted/Chunks/Textures/T_Abducted_Sediment_B}
    TextureSize=ScalarParameter() {defaultValue=5000, group=Sediment, parameterName=Sediment Tiling}
   ) {materialFunction=&WorldAlignedTexture}.XY Texture
  )
  n61 = ScalarParameter() {defaultValue=0.4, group=Sediment, parameterName=Sediment Alpha}
  n62 = StaticSwitchParameter(
   True=NamedRerouteUsage()
   False=NamedRerouteDeclaration(None=MaterialFunctionCall!(Material=@n59.Color*Gradient, Tint=@n60, TintMask=@n61) {materialFunction=/Engine/Functions/MaterialLayerFunctions/MatLayerBlend_Tint}.Blended Material) {name=Set1, nodeColor=<rgba 0 0.4141 1 1>}
  ) {desc=Uses Vertex color to mask. If can - Don't use, group=Set2, parameterName=Set2}
  n63 = GetMaterialAttributes(None=@n62) {attributeGetTypes [
    #1
    #2
    #3
    #5
    #6
   ]}
  n64 = ShadingPathSwitch(Default=@n63.MaterialAttributes, Deferred=@n63.MaterialAttributes, Forward=@n62, Mobile=@n62)
  n65 = NamedRerouteDeclaration(None=DistanceFieldApproxAO!(
    World Position
    Normal
    BaseDistance=ScalarParameter() {defaultValue=200, group=DFAO, parameterName=DFAO_Distance}
    Radius=ScalarParameter() {defaultValue=350, group=DFAO, parameterName=DFAO_Radius}
   ) {numSteps=2}) {name=AO, nodeColor=<rgba 0.2656 1 0 1>}
  n66 = MaterialFunctionCall(
   Material=@n64
   Tint=VectorParameter() {group=DFAO, parameterName=DFAO_Color, defaultValue=<rgba 0.5156 0.3855 0.3384 1>}.RGB
   TintMask=OneMinus(None=Saturate(None=@n65))
  ) {materialFunction=&MatLayerBlend_Tint}
  n67 = FeatureLevelSwitch(
   Default=@n66.Blended Material
   ES2=@n66.Blended Material
   ES3_1=@n64
   SM4_REMOVED=@n66.Blended Material
   SM5=@n66.Blended Material
   SM6=@n66.Blended Material
  )
  n68 = BlendMaterialAttributes(A=NamedRerouteUsage(), B=@n67, Alpha=NamedRerouteUsage()) {vertexAttributeBlendType=UseA}
  n69 = StaticSwitchParameter(True=@n67, False=ShadingPathSwitch(Default=@n68, Deferred=@n68, Forward=@n67, Mobile=@n67)) {desc=Uses Vertex color to mask. If can - Don't use, group=Transition Texture, parameterName=SkipTransition Textures}
  n70 = TextureSampleParameter2D() {parameterName=Baked Normal, samplerType=SAMPLERTYPE_Normal, texture=/Game/Environments/Asteria/Sets/Quarry/Textures/T_Quarry_Cube_2x2x2_D_N}
  n71 = NaniteReplace(Default=@n70.RGB, Nanite=Constant3Vector() {constant=<rgba 0 0 1 0>})
  n72 = MaterialFunctionCall!(Material=StaticSwitchParameter(
    True=MaterialFunctionCall!(Material=@n69, Normal=@n71) {materialFunction=/Engine/Functions/MaterialLayerFunctions/MatLayerBlend_BakedNormal}.Blended Material
    False=@n69
   ) {parameterName=Nanite Test}) {materialFunction=/Game/Athena/Apollo/Maps/Landscape/VT/MaterialFunctions/TransformTangentToWorld}
  n73 = RuntimeVirtualTextureSampleParameter() {materialType=BaseColor_Normal_Specular_YCoCg, parameterName=RVT, virtualTexture=/Game/Athena/Apollo/Maps/Landscape/VT/RVT_Apollo_Landscape}
  n74 = WorldPosition()
  n75 = RuntimeVirtualTextureSampleParameter() {materialType=WorldHeight, parameterName=RVTHeight, virtualTexture=/Game/Athena/Apollo/Maps/Landscape/VT/RVT_Apollo_Landscape_Height}
  n76 = Subtract(A=Add(A=ScalarParameter() {group=RVT, parameterName=Offset RVT WP}, B=ComponentMask(None=@n74.XYZ) {b=true}), B=@n75.WorldHeight)
  n77 = Divide(
   A=NamedRerouteDeclaration(None=@n76) {name=RVT Height, nodeColor=<rgba 0 0.8594 1 1>}
   B=ScalarParameter() {defaultValue=25, group=RVT, parameterName=Blend RVT Normals}
  )
  n78 = DotProduct(A=VertexNormalWS(), B=Constant3Vector() {constant=<rgba 0 0 1 0>})
  n79 = VertexNormalWS()
  n80 = Constant3Vector() {constant=<rgba 0 0 -1 0>}
  n81 = StaticSwitchParameter(True=OneMinus(None=Clamp(None=@n78)), False=Clamp(None=DotProduct(A=@n79, B=@n80))) {group=RVT, parameterName=RVT top only?}
  n82 = BlendMaterialAttributes(
   A=SetMaterialAttributes!(MaterialAttributes=@n72.Result, Normal=@n73.Normal) {attributeSetTypes [#4]}
   B=@n72.Result
   Alpha=LinearInterpolate(A=Saturate(None=@n77), Alpha=@n81)
  ) {vertexAttributeBlendType=UseA}
  n83 = GetMaterialAttributes(None=@n72.Result) {attributeGetTypes [#1, #4]}
  n84 = ComponentMask(None=VertexNormalWS()) {b=true}
  n85 = SetMaterialAttributes!(
   MaterialAttributes=@n82
   Base Color=StaticSwitchParameter(True=@n73.BaseColor, False=LinearInterpolate(A=@n83.Base Color, B=@n73.BaseColor, Alpha=@n84)) {parameterName=Skip Side/Up mask}
   Specular=@n73.Specular
   Roughness=@n73.Specular
   Emissive Color=Constant()
  ) {attributeSetTypes [
    #1
    #2
    #3
    #7
   ]}
  n86 = ScalarParameter() {defaultValue=55, group=RVT, parameterName=Blend RVT}
  n87 = LinearInterpolate(A=Saturate(None=Divide(A=@n76, B=@n86)), Alpha=@n81)
  n88 = MaterialFunctionCall(
   NoRVT=@n72.Result
   RVT=MaterialProxyReplace(Realtime=BlendMaterialAttributes(A=@n85, B=@n82, Alpha=@n87) {vertexAttributeBlendType=UseA}, MaterialProxy=@n72.Result)
  ) {materialFunction=&FN_RVT_Switch_Material}
  StaticSwitchParameter(True=@n72.Result, False=ShadingPathSwitch(Default=@n88.Result, Deferred=@n88.Result, Forward=@n72.Result, Mobile=@n72.Result)) {
   desc="If for some reason you want to use this material but hate to use RVT ( or you'r Bryce L) Will automatically convert details to use up vector instead distance from the ground "
   group=RVT
   parameterName=Skip RVT
   sortPriority=-4
  }
 ]
}
